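# Level sequence for the QuartzSqueak custom skin's Battle Royale menu animation, under the season 39 folder.
# A spawned FortPlayerMannequin at the origin plays a skeletal animation over frames 0–360.
# A legacy face Control Rig track on the mannequin holds its master control, with constant values on the channels shown.
# The child Head mesh carries a face bone-tweaker Control Rig track with no channels shown.
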LevelSequence /BRCosmetics/Animation/Game/MainPlayer/Menu/BR/CustomSkins/_s39/QuartzSqueak/QuartzSqueak_LS
bindings [
 Head {
  id=#1
  parent_id=#2
  possessed_class=SkeletalMeshComponentBudgeted
  tracks [{
    type=MovieSceneControlRigParameterTrack
    display_name=Fortnite_Face_BoneTweaker_CtrlRig
    sections [{type=MovieSceneControlRigParameterSection}]
   }]
 }
 FortPlayerMannequin {
  id=#2
  possessed_class=FortPlayerMannequin
  tracks [
   {type=MovieSceneSpawnTrack, sections [{type=MovieSceneSpawnSection, channels [None {type=MovieSceneScriptingBoolChannel, default=true}]}]}
   {
    type=MovieScene3DTransformTrack
    sections [{
      type=MovieScene3DTransformSection
      channels [
       Location.X {type=MovieSceneScriptingDoubleChannel, default=0}
       Location.Y {type=MovieSceneScriptingDoubleChannel, default=0}
       Location.Z {type=MovieSceneScriptingDoubleChannel, default=0}
       Rotation.X {type=MovieSceneScriptingDoubleChannel, default=0}
       Rotation.Y {type=MovieSceneScriptingDoubleChannel, default=0}
       Rotation.Z {type=MovieSceneScriptingDoubleChannel, default=0}
       Scale.X {type=MovieSceneScriptingDoubleChannel, default=1}
       Scale.Y {type=MovieSceneScriptingDoubleChannel, default=1}
       Scale.Z {type=MovieSceneScriptingDoubleChannel, default=1}
      ]
     }]
   }
   {type=MovieSceneSkeletalAnimationTrack, sections [{type=MovieSceneSkeletalAnimationSection, end_frame=360, start_frame=0}]}
   {
    type=MovieSceneControlRigParameterTrack
    display_name=Fortnite_Legacy_Face_CtrlRig
    sections [{
      type=MovieSceneControlRigParameterSection
      channels [
       Flipbook {type=MovieSceneScriptingBoolChannel, default=true, const [false,1]}
       Face_Ctrl_Master.Location.X {type=MovieSceneScriptingFloatChannel, default=136.3155, const [136.3155,361]}
       Face_Ctrl_Master.Location.Y {type=MovieSceneScriptingFloatChannel, default=-0.9823, const [-0.9823,361]}
       Face_Ctrl_Master.Location.Z {type=MovieSceneScriptingFloatChannel, default=-118.8415, const [-118.8415,361]}
       Face_Ctrl_Master.Rotation.X {type=MovieSceneScriptingFloatChannel, default=0, const [0,361]}
       Face_Ctrl_Master.Rotation.Y {type=MovieSceneScriptingFloatChannel, default=0, const [0,361]}
       Face_Ctrl_Master.Rotation.Z {type=MovieSceneScriptingFloatChannel, default=0, const [0,361]}
       Face_Ctrl_Master.Scale.X {type=MovieSceneScriptingFloatChannel, default=9.1564, const [9.1564,361]}
       Face_Ctrl_Master.Scale.Y {type=MovieSceneScriptingFloatChannel, default=9.1564, const [9.1564,361]}
       Face_Ctrl_Master.Scale.Z {type=MovieSceneScriptingFloatChannel, default=9.1564, const [9.1564,361]}
       gui_follow_head {type=MovieSceneScriptingFloatChannel, default=3.4028234663852886e+38, const [1,1]}
      ]
     }]
   }
  ]
 }
]
display_rate {denominator=1, numerator=30}
tick_resolution {denominator=1, numerator=24000}
tracks []
